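# StrawTurn quest brief widget, a FortQuestBrief child with no widgets of its own, that presents Jigsaw's Challenge in the conversation UI.
# PreConstruct sets a warning about elimination on time-out and tints it yellow.
# When quest information arrives, it calls the parent with one token reward and a StrawTurn icon, then overwrites the title and transaction strip's cost description with challenge and rewards text.
# The SetWarningText override only forwards to the parent.
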WidgetBlueprint /StrawTurn/Widgets/FortStrawTurnJobBrief
graphs [
 EventGraph {nodes [
   Event PreConstruct()
   > SetWarningText!(execute, self:Fort Quest Brief Object Reference, WarningText:Text=Get StrawTurnWarningText() -> (StrawTurnWarningText:Text)) -> (then)
   > Set ColorAndOpacity!(
    execute
    ColorAndOpacity:Linear Color Structure=(R=1,G=0.9301,B=0,A=1)
    self:User Widget Object Reference=Get WarningMessage() -> (WarningMessage:Brief Error Message New Object Reference)
   ) -> (then, Output_Get:Linear Color Structure)
   n1 = Event OnQuestInformationRecieved!() -> (
    OutputDelegate
    then
    Objective:"Text (by ref)"
    RarityData:"Fort Rarity Item Data Structure (by ref)"
    CategoryTag:"Gameplay Tag Structure (by ref)"
    XPReward:int=0
    MoneyReward:int=0
    TokenReward:int=0
    TokenRewardIcon:"Texture 2D Soft Object Reference (by ref)"
   ) {is=title K2Node_Event}
   > Parent: OnQuestInformationRecieved!(
    execute
    Objective:"Text (by ref)"=@n1.Objective
    RarityData:"Fort Rarity Item Data Structure (by ref)"=@n1.RarityData
    CategoryTag:"Gameplay Tag Structure (by ref)"=@n1.CategoryTag
    XPReward:int=@n1.XPReward
    MoneyReward:int=@n1.MoneyReward
    TokenReward:int=1
    TokenRewardIcon:"Texture 2D Soft Object Reference (by ref)"=Get Token Reward Icon() -> (Token Reward Icon:Texture 2D Soft Object Reference)
   ) -> (then) {is=title K2Node_CallParentFunction}
   > "SetText (Text)"(self=Get Title() -> (Title:Common Text Object Reference), InText="LOCTEXT(\"JIGSAW\\'S CHALLENGE\")")
   > "SetText (Text)"(
    self=Get Text_CostDescription!(self:Fort Transaction Strip Object Reference=Get FortTransactionStrip() -> (FortTransactionStrip:Fort Transaction Strip New Object Reference)) -> (Text_CostDescription:Common Text Object Reference)
    InText=Get StrawTurnRewardsText() -> (StrawTurnRewardsText:Text)
   )
   Event Construct()
   Event Tick()
  ]}
 SetWarningText {nodes [
   n1 = SetWarningText!() -> (then, WarningText:Text) {is=title K2Node_FunctionEntry}
   > Parent: SetWarningText!(execute, WarningText:Text=@n1) -> (then) {is=title K2Node_CallParentFunction}
  ]}
]
variables [
 StrawTurnRewardsText {type=string, default=REWARDS WEAPONS AND ITEMS}
 StrawTurnWarningText {type=string, default="WARNING: JIGSAW'S DEVICE WILL ELIMINATE YOU IF YOUR OPPONENT IS NOT ELIMINATED IN THE TIME LIMIT!"}
 Token Reward Icon {type=Texture2D, default=/StrawTurn/UI/T_UI_Icon_Reward_StrawTurn}
]
widgets {
 info {
  inheritedWidgetCount=0
  parentClass=/FortniteConversationUI/Conversation/FortQuestBrief.FortQuestBrief_C
  rootWidgetClass=null
  widgetCount=0
 }
 widgets []
}
