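# Conversation database for the StrawTurn job board briefing, offering a player duel from the JobBoard.PlayerDuel entry point.
# A speech node branches to Cancel, the duel offer and unavailable-reason choices, gated by service, urgent quest, active-quest tags and recorded grant history.
# Accepting records the GrantedQuest event, deactivates the board and starts the duel with StrawTurn's quest and gameplay effects.
# That recorded event later shows "No Briefings Available"; otherwise, active job-board or bounty quests show "Briefing Already Active".
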ConversationDatabase /StrawTurn/Conversations/Conversation_StrawTurn_Briefing
entry_points [{tag=Conversation.Entry.JobBoard.PlayerDuel, destinations [#1]}]
nodes [
 ConversationEntryPointNode_6 {
  id=#1
  class=ConversationEntryPointNode
  outputs [#2]
  properties {entryTag {tagName=Conversation.Entry.JobBoard.PlayerDuel}}
 }
 Speech_C_3 {
  id=#2
  class=Speech_C
  outputs [
   #3
   #4
   #5
   #6
   #7
  ]
  properties {nodeName=""}
 }
 ShowUserDataDrivenCannotUseReason_C_0 {
  id=#7
  class=ShowUserDataDrivenCannotUseReason_C
  outputs [#8]
  properties {
   cannotUseReason=BriefingAlreadyActive
   hoverChoiceSound=/JobBoard/Sounds/JobBoard_ShadowBriefing_NotAvailable_Cue
   serviceBriefCollectionKeyOveride=Service.JobBoard.DisplayCannotUseReason
   serviceBriefConfigCollection=/JobBoard/Conversations/JobBoardServiceBriefCollectionConfig
   softTaskIcon=&Cancel_Normal
  }
  sub_nodes [
   ConversationChoiceNode_16 {
    class=ConversationChoiceNode
    properties {defaultChoiceDisplayText=BRIEFING ALREADY ACTIVE, choiceTags {gameplayTags [{tagName=Conversation.Choice.Category.Back}]}}
   }
   FortConversationRequirement_ControllerRequirement_25 {
    class=FortConversationRequirement_ControllerRequirement
    properties {
     nodeName=Require: Has Recorded Conversation Event With NPC
     requirement {
      class=FortControllerRequirement_HasRecordedConversationEventWithNPC
      properties {bInverseCondition=true, queryMethod=CheckAgainstCurrentConversationParticipant, queryEventTag {tagName=Conversation.Event.GrantedQuest}}
     }
    }
   }
   FortConversationRequirement_ControllerRequirement_26 {
    class=FortConversationRequirement_ControllerRequirement
    properties {
     nodeName=Require: Has GameplayTag
     requirement {class=FortControllerRequirement_HasOwnedTag, properties {ownedTag {tagName=Granted.Athena.BountyBoard.QuestActive}}}
    }
   }
  ]
 }
 Continue_C_9 {id=#8, class=Continue_C}
 ShowUserDataDrivenCannotUseReason_C_7 {
  id=#6
  class=ShowUserDataDrivenCannotUseReason_C
  outputs [#9]
  properties {
   cannotUseReason=BriefingAlreadyActive
   hoverChoiceSound=&JobBoard_ShadowBriefing_NotAvailable_Cue
   serviceBriefCollectionKeyOveride=Service.JobBoard.DisplayCannotUseReason
   serviceBriefConfigCollection=&JobBoardServiceBriefCollectionConfig
   softTaskIcon=&Cancel_Normal
  }
  sub_nodes [
   ConversationChoiceNode_15 {
    class=ConversationChoiceNode
    properties {defaultChoiceDisplayText=BRIEFING ALREADY ACTIVE, choiceTags {gameplayTags [{tagName=Conversation.Choice.Category.Back}]}}
   }
   FortConversationRequirement_ControllerRequirement_19 {
    class=FortConversationRequirement_ControllerRequirement
    properties {
     nodeName=Require: Has Recorded Conversation Event With NPC
     requirement {
      class=FortControllerRequirement_HasRecordedConversationEventWithNPC
      properties {bInverseCondition=true, queryMethod=CheckAgainstCurrentConversationParticipant, queryEventTag {tagName=Conversation.Event.GrantedQuest}}
     }
    }
   }
   FortConversationRequirement_ControllerRequirement_20 {
    class=FortConversationRequirement_ControllerRequirement
    properties {
     nodeName=Require: Has GameplayTag
     requirement {class=FortControllerRequirement_HasOwnedTag, properties {ownedTag {tagName=Granted.Athena.JobBoardExperimental.QuestActive}}}
    }
   }
  ]
 }
 Continue_C_5 {id=#9, class=Continue_C}
 ShowUserDataDrivenCannotUseReason_C_1 {
  id=#5
  class=ShowUserDataDrivenCannotUseReason_C
  outputs [#10]
  properties {
   briefWidgetClassOverride=/FortniteConversationUI/Conversation/FortDataDrivenServiceBrief.FortDataDrivenServiceBrief_C
   cannotUseReason=NoBriefingsAvailable
   hoverChoiceSound=&JobBoard_ShadowBriefing_NotAvailable_Cue
   serviceBriefCollectionKeyOveride=Service.JobBoard.DisplayCannotUseReason
   serviceBriefConfigCollection=&JobBoardServiceBriefCollectionConfig
   softTaskIcon=&Cancel_Normal
  }
  sub_nodes [
   ConversationChoiceNode_11 {
    class=ConversationChoiceNode
    properties {defaultChoiceDisplayText=NO BRIEFINGS AVAILABLE, choiceTags {gameplayTags [{tagName=Conversation.Choice.Category.Back}]}}
   }
   FortConversationRequirement_ControllerRequirement_13 {
    class=FortConversationRequirement_ControllerRequirement
    properties {
     nodeName=Require: Has Recorded Conversation Event With NPC
     requirement {
      class=FortControllerRequirement_HasRecordedConversationEventWithNPC
      properties {queryMethod=CheckAgainstCurrentConversationParticipant, queryEventTag {tagName=Conversation.Event.GrantedQuest}}
     }
    }
   }
  ]
 }
 Continue_C_7 {id=#10, class=Continue_C}
 GrantSlottedQuest_C_4 {
  id=#4
  class=GrantSlottedQuest_C
  outputs []
  properties {
   bSharedQuestLimitsEnforced=false
   bShouldOverrideSpeakerText=false
   briefWidgetClassOverride=/StrawTurn/Widgets/FortStrawTurnJobBrief.FortStrawTurnJobBrief_C
   choiceSound=null
   softTaskIcon=/Game/UI/Foundation/Textures/Icons/Checks/T-Icon-Check-128
   optionalFilterCategoryTag {tagName=Tandem.Service.StrawTurn.PlayerDuel}
  }
  sub_nodes [
   ConversationChoiceNode_3 {class=ConversationChoiceNode, properties {defaultChoiceDisplayText=DO YOU WANT TO PLAY A GAME?}}
   FortConversationRequirement_HasService_0 {class=FortConversationRequirement_HasService, properties {serviceTag {tagName=Tandem.Service.StrawTurn.PlayerDuel}}}
   FortConversationRequirement_ControllerRequirement_2 {
    class=FortConversationRequirement_ControllerRequirement
    properties {
     nodeName=Require: Has Urgent Quest
     requirement {
      class=FortControllerRequirement_HasUrgentQuest
      properties {
       bInverseCondition=true
       bRequireExactMatch=false
       acceptedUrgentEventTags {gameplayTags [{tagName=SpecialEvent.HuntingPlayer.IsHunter}, {tagName=SpecialEvent.JobBoard.Experimental}, {tagName=SpecialEvent.StrawTurn.PlayerDuel}]}
      }
     }
    }
   }
   FortConversationRequirement_ControllerRequirement_3 {
    class=FortConversationRequirement_ControllerRequirement
    properties {
     nodeName=Require: Has GameplayTag
     requirement {
      class=FortControllerRequirement_HasOwnedTag
      properties {bInverseCondition=true, ownedTag {tagName=Granted.Athena.JobBoardExperimental.QuestActive}}
     }
    }
   }
   FortConversationRequirement_ControllerRequirement_4 {
    class=FortConversationRequirement_ControllerRequirement
    properties {
     nodeName=Require: Has GameplayTag
     requirement {class=FortControllerRequirement_HasOwnedTag, properties {bInverseCondition=true, ownedTag {tagName=Granted.Athena.BountyBoard.QuestActive}}}
    }
   }
   SideEffect_PlayServiceUsedEffects_C_8 {class=SideEffect_PlayServiceUsedEffects_C}
   SideEffect_Server_RecordEventToHistoryData_C_8 {class=SideEffect_Server_RecordEventToHistoryData_C, properties {event Type {tagName=Conversation.Event.GrantedQuest}}}
   SideEffect_Server_TrackJobBoardUser_C_8 {class=SideEffect_Server_TrackJobBoardUser_C}
   SideEffect_Server_JobBoardExperimental_DeactivateOnUse_C_3 {class=SideEffect_Server_JobBoardBriefing_DeactivateOnUse_C}
   SideEffect_Server_JobBoardBriefing_ApplyInstigatorGE_C_0 {class=SideEffect_Server_JobBoardBriefing_ApplyInstigatorGE_C}
   SideEffect_Server_JobBoardExperimental_StartDuel_C_3 {
    class=SideEffect_Server_StartPlayerDuel_C
    properties {
     duel Quest Definition=/StrawTurn/Items/QuestItems/quest_s31_strawturn_urgent_qs1
     optionalDuelInstigatorGE=/StrawTurn/Gameplay/Abilities/GE_Granted_StrawTurn_PlayerDuel_Override.GE_Granted_StrawTurn_PlayerDuel_Override_C
     phoebe Gameplay Effect Class=/StrawTurn/Gameplay/Abilities/GE_PlayerDuel_Phoebe_StrawTurn.GE_PlayerDuel_Phoebe_StrawTurn_C
    }
   }
  ]
 }
 Continue_C_8 {
  id=#3
  class=Continue_C
  properties {softTaskIcon=&Cancel_Normal}
  sub_nodes [
   ConversationChoiceNode_13 {
    class=ConversationChoiceNode
    properties {defaultChoiceDisplayText=CANCEL, choiceTags {gameplayTags [{tagName=Conversation.Choice.CustomSortPriority.Highest}]}}
   }
   FortConversationRequirement_ControllerRequirement_15 {
    class=FortConversationRequirement_ControllerRequirement
    properties {
     nodeName=Require: Has GameplayTag
     requirement {
      class=FortControllerRequirement_HasOwnedTag
      properties {bInverseCondition=true, ownedTag {tagName=Granted.Athena.JobBoardExperimental.QuestActive}}
     }
    }
   }
   FortConversationRequirement_ControllerRequirement_16 {
    class=FortConversationRequirement_ControllerRequirement
    properties {
     nodeName=Require: Has GameplayTag
     requirement {class=FortControllerRequirement_HasOwnedTag, properties {bInverseCondition=true, ownedTag {tagName=Granted.Athena.BountyBoard.QuestActive}}}
    }
   }
  ]
 }
]
